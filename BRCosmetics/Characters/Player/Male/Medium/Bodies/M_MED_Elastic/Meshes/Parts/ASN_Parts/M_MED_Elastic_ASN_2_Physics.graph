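# Physics asset for the M_MED_Elastic_ASN_2 character part, under the male medium player body cosmetics.
# It defines no constraints and shows no physics bodies, so it appears to be an empty placeholder with no simulation setup of its own.
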PhysicsAsset /BRCosmetics/Characters/Player/Male/Medium/Bodies/M_MED_Elastic/Meshes/Parts/ASN_Parts/M_MED_Elastic_ASN_2_Physics
constraints []
